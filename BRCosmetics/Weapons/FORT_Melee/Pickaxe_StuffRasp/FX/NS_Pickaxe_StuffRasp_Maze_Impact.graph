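# Niagara impact effect for the StuffRasp Maze pickaxe, layering monochrome sprite bursts of glow, energy smoke, claw slash and spikes; physical-material debris is disabled.
# User.Variant blends from bright to dark: 0 adds claw sparks, 1 adds black smoke.
# User.SwingSectionIndex steers claw direction, and section 3 doubles claw and spike counts, split two ways.
# Spawn groups let emitters share sub-effects, while front-end preview flags suppress or rescale layers.
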NiagaraSystem /BRCosmetics/Weapons/FORT_Melee/Pickaxe_StuffRasp/FX/NS_Pickaxe_StuffRasp_Maze_Impact
dependencies {
 usedDataInterfaces [
  &NiagaraDataInterfaceCurve
  &NiagaraDataInterfaceColorCurve
  &NiagaraDataInterfaceArrayFloat3
  &NiagaraDataInterfaceVector2DCurve
  &NiagaraDataInterfaceSpriteRendererInfo
  &NiagaraDataInterfaceArrayFloat
 ]
 usedDynamicInputs [
  &MakeVector
  &ReturnNormalizedExecIndex
  &SelectFloatFromArray
  &NormalizeVector
  /Game/Effects/Niagara/DynamicInputs/FN_DirectionalVectorfromMontageIndex
  &ModuloFloat
  &FN_FrontEndVector3Modifier
  /Game/Effects/Niagara/Modules/Utility/FN_DirectionVectorByPickaxeType
 ]
 usedModules [
  &self:SystemSpawnScript.SystemScriptSource.SystemScriptGraph.NiagaraNodeAssignment_0.SetVariables_#1
  &self:SystemSpawnScript.SystemScriptSource.SystemScriptGraph.NiagaraNodeAssignment_1.SetVariables_#2
  &self:SystemSpawnScript.SystemScriptSource.SystemScriptGraph.NiagaraNodeAssignment_2.SetVariables_#3
  &self:SystemSpawnScript.SystemScriptSource.SystemScriptGraph.NiagaraNodeAssignment_3.SetVariables_#4
  &SystemState
 ]
}
emitters {
 Center_spikes {
  emitterUpdateScript [
   &EmitterState(Life Cycle Mode=::NewEnumerator1, Loop Behavior=::NewEnumerator1, Loop Delay=0.03, Loop Duration=0.5, UseLoopDelay=false)
   &SpawnBurst_Instantaneous(
    Spawn Count=&MakeCustomIntFromBool(
     Bool=&SetBoolByIntComparison(A=$User.SwingSectionIndex, B=3, Comparison Type=::NewEnumerator4)
     True INT=16
     False INT=8
    )
    Spawn Probability=&MakeCustomFloatFromBool(Bool=$User.bIsFrontEndPreview, True Float=0, False Float=1)
    Use Spawn Probability=true
    UseLoopCountLimit=false
   )
  ]
  particleSpawnScript [
   &PartitionParticles()
   &InitializeParticle(
    Color=&LerpLinearColors(
     StartColor=&RandomRangeLinearColor(Minimum=<rgba 0.0885 0.0885 0.0885 3>, Maximum=<rgba 0.1771 0.1771 0.1771 1.5>)
     EndColor=&RandomRangeLinearColor(Minimum=<rgba 0.01 0.01 0.01 3>, Maximum=<rgba 0.02 0.02 0.02 1.5>)
     LerpFactor=$User.Variant
    )
    Lifetime Max=0.16
    Lifetime Min=0.12
    Lifetime Mode=::NewEnumerator1
    Position Offset=<10 0 0>
    Sprite Size=&Multiply_Vector2D(
     A=&Vector2DFromFloat(Value=$System.FEScale)
     B=&RandomRangeVector2D(Minimum=<6 100>, Maximum=<10 125>)
    )
    Sprite Size Mode=::NewEnumerator4
    UsePositionOffset=false
    Write Material Random=true
   )
   &ShapeLocation(
    MaskSpawnGroup=false
    Offset=&MakeCustomVectorFromBool(Bool=false, True Vector=<-10 0 0>, False Vector=<0 0 0>)
    Offset Mode=::NewEnumerator0
    Override Seed=false
    Sphere Radius=2
    UseOwnerScale=true
   )
   &SubUVAnimation(End Frame=0, SubUV Animation Mode:ENiagara_SubUVLookupMode=::NewEnumerator2)
   &AddVelocity(
    Cone Axis=&Lerp_Vector(
     A=$System.DirectionalVector
     B=&SelectVectorFromArray(
      Direct Array Index=$Particles.PartitionParticles.Partition
      Vector Selection Array=DataInterface() {propertyValues {FloatData [<-1 0 -0.7>, <1 0 -0.7>]}}
     )
     Alpha=$System.isLastSwing
    )
    Module.UseConeFalloff=false
    Module.UseVelDistribution=false
    Override Seed=false
    Velocity Mode=::NewEnumerator2
    Velocity Speed=&RandomRangeFloat(Minimum=225, Maximum=400)
   )
   &CameraOffset(Camera Offset Amount=10)
  ]
  particleUpdateScript [
   &ParticleState()
   &Drag(Drag=&RandomRangeFloat(Minimum=0.8, Maximum=1.2), Use Linear Drag=true, Use Rotational Drag=false)
   &SolveForcesAndVelocity(Clamp Velocity=false, Limit Acceleration=false)
   &ScaleSpriteSize(
    Non-Uniform Curve Scale=&Multiply_Vector2D_ByFloat(Float=1, Vector2D=<1.1 1.3>)
    Non-Uniform Curve Sprite Scale=DataInterface() {propertyValues {
      XCurve {keys [{tangentMode=RCTM_User, time=0, value=1}, {tangentMode=RCTM_User, time=1, value=1}]}
      YCurve {keys [{tangentMode=RCTM_User, time=0, value=0.2}, {tangentMode=RCTM_User, time=1, value=1.2}]}
     }}
    Scale Sprite Size Mode=::NewEnumerator3
   )
   &ScaleColor(
    Scale Alpha=&FloatFromCurve(
     FloatCurve=DataInterface() {propertyValues {Curve {keys [
         <at 0 0 RCIM_Cubic>
         {arriveTangent=-0.512, interpMode=RCIM_Cubic, leaveTangent=-0.512, tangentMode=RCTM_User, time=0.15, value=1}
         <at 1 0 RCIM_Cubic>
        ]}}}
     Scale Curve=2.2
    )
    Scale RGB=&MakeVectorFromLinearColorRGB(Color=<rgba 1 1 1 1>)
    ScaleA=true
    ScaleRGB=true
   )
  ]
  rendererClasses [&NiagaraSpriteRendererProperties]
  renderers [{rendererClass=&NiagaraSpriteRendererProperties}]
 }
 Claw {
  emitterUpdateScript [
   &EmitterState(Life Cycle Mode=::NewEnumerator1, Loop Behavior=::NewEnumerator1, Loop Duration=1, UseLoopDelay=false)
   &SpawnBurst_Instantaneous(
    Spawn Count=&MakeCustomIntFromBool(
     Bool=&SetBoolByFloatComparison(A=$User.Variant, B=0, Comparison Type=::NewEnumerator4)
     True INT=&MakeCustomIntFromBool(
      Bool=&SetBoolByIntComparison(A=$User.SwingSectionIndex, B=3, Comparison Type=::NewEnumerator4)
      True INT=48
      False INT=24
     )
     False INT=0
    )
    Spawn Group=1
    Use Spawn Probability=false
    UseLoopCountLimit=false
   ) {module=SpawnBurst_Instantaneous001}
   &SpawnBurst_Instantaneous(
    Spawn Count=&MakeCustomIntFromBool(
     Bool=&SetBoolByIntComparison(A=$User.SwingSectionIndex, B=3, Comparison Type=::NewEnumerator4)
     True INT=2
     False INT=1
    )
    Use Spawn Probability=false
    UseLoopCountLimit=false
   )
  ]
  particleSpawnScript [
   /SetVariables(Particles.VisibilityTag=&SpawnGroup()) {isSetParameters=true}
   &PartitionParticles()
   &InitializeParticle(
    Color=&MaskLinearColorBySpawnGroup(
     Masked Linear Color=&LerpLinearColors(StartColor=<rgba 1.4 1.4 1.4 1>, EndColor=<rgba 1 1 1 1>, LerpFactor=$User.Variant)
     Passthrough Linear Color=&RandomRangeLinearColor(Minimum=<rgba 3 3 3 1>, Maximum=<rgba 1 1 1 1>)
    )
    Lifetime=&MaskFloatBySpawnGroup(Masked Float=0.18, Passthrough Float=&RandomRangeFloat(Minimum=0.3, Maximum=0.5))
    Position Offset=&MakeCustomVectorFromBool(
     Bool=&SetBoolByIntComparison(A=$User.SwingSectionIndex, B=2, Comparison Type=::NewEnumerator4)
     True Vector=<-15 0 -5>
     False Vector=<0 0 0>
    )
    Sprite Size=&MaskVector2dBySpawnGroup(
     Masked Vector2D=&Multiply_Vector2D_ByFloat(Float=$System.FEScale, Vector2D=<-60 120>)
     Passthrough Vector2D=&Vector2DFromFloat(Value=&RandomRangeFloat(Minimum=0.75, Maximum=2.25))
     Spawn Group Mask=0
    )
    Sprite Size Mode=::NewEnumerator4
    UsePositionOffset=true
    Write Material Random=true
   )
   &CameraOffset(Camera Offset Amount=&FN_FrontEndFloatModifier(Gameplay=50, Front End=0))
   &ShapeLocation(
    MaskSpawnGroup=false
    Override Seed=false
    Sphere Radius=&MaskFloatBySpawnGroup(Masked Float=0, Passthrough Float=12)
    UseOwnerScale=true
   )
   &AddVelocity(Velocity=&MaskVectorBySpawnGroup(
     Masked Vector=&MakeCustomVectorFromBool(
      Bool=$User.bIsFrontEndPreview
      True Vector=&MakeCustomVectorFromBool(
       Bool=&SetBoolByIntComparison(A=$User.SwingSectionIndex, B=2, Comparison Type=::NewEnumerator4)
       True Vector=<100 50 125>
       False Vector=&MakeCustomVectorFromBool(
        Bool=&SetBoolByIntComparison(A=$User.SwingSectionIndex, B=2, Comparison Type=::NewEnumerator4)
        True Vector=&SelectVectorFromArray(
         Direct Array Index=$Particles.PartitionParticles.Partition
         Vector Selection Array=DataInterface() {propertyValues {FloatData [<25 0 -50>, <-50 100 -50>]}}
        )
        False Vector=<-50 100 -50>
       )
      )
      False Vector=&MakeCustomVectorFromBool(
       Bool=&SetBoolByIntComparison(A=$User.SwingSectionIndex, B=2, Comparison Type=::NewEnumerator4)
       True Vector=<125 -10 175>
       False Vector=&MakeCustomVectorFromBool(
        Bool=&SetBoolByIntComparison(A=$User.SwingSectionIndex, B=3, Comparison Type=::NewEnumerator4)
        True Vector=&SelectVectorFromArray(
         Direct Array Index=$Particles.PartitionParticles.Partition
         Vector Selection Array=DataInterface() {propertyValues {FloatData [<150 10 -70>, <-150 10 -70>]}}
        )
        False Vector=<250 10 -30>
       )
      )
     )
     Passthrough Vector=<0 0 0>
    )) {module=AddVelocity001}
   &AddVelocity(
    Cone Angle=60
    Cone Axis=&MakeCustomVectorFromBool(
     Bool=&SetBoolByFloatComparison(A=&MakeFloatFromInt(Int=$User.SwingSectionIndex), B=3, Comparison Type=::NewEnumerator4)
     True Vector=&MaskVector(Mask Source=1, Mask Parameter=$Particles.PartitionParticles.Partition, Mask=0, Masked Vector=<1 0 -0.5>, Passthrough Vector=<-1 0 -0.5>)
     False Vector=&SelectVectorFromArray(
      Direct Array Index=$User.SwingSectionIndex
      Vector Selection Array=DataInterface() {propertyValues {FloatData [<1 0 -0.5>, <1 0 -0.5>, <1 0 1>]}}
     )
    )
    Module.UseConeFalloff=false
    Module.UseVelDistribution=false
    Override Seed=false
    Velocity Mode=::NewEnumerator2
    Velocity Speed=&MaskFloatBySpawnGroup(Masked Float=0, Passthrough Float=&RandomRangeFloat(Minimum=525, Maximum=700))
   )
  ]
  particleUpdateScript [
   &ParticleState()
   &ScaleSpriteSize(
    Scale Factor=&MaskVector2dBySpawnGroup(
     Masked Vector2D=&Vector2DFromCurve(
      Vector2Curve=DataInterface() {propertyValues {
        XCurve {keys [
          {arriveTangent=-0.1132, interpMode=RCIM_Cubic, leaveTangent=-0.1132, tangentMode=RCTM_User, time=0, value=2}
          {arriveTangent=-1.4, interpMode=RCIM_Cubic, leaveTangent=-1.4, time=0.6523, value=1.6775}
          {arriveTangent=-0.1428, interpMode=RCIM_Cubic, leaveTangent=-0.1428, tangentMode=RCTM_User, time=1, value=0.6}
         ]}
        YCurve {keys [
          {arriveTangent=13.6387, interpMode=RCIM_Cubic, leaveTangent=13.6387, tangentMode=RCTM_User, time=0.005, value=0.2241}
          {arriveTangent=0.7768, tangentMode=RCTM_User, time=1, value=4.4969}
         ]}
       }}
      ScaleCurve=&Multiply_Vector2D_ByFloat(Float=0.25, Vector2D=<1 1>)
     )
     Passthrough Vector2D=&Vector2DFromCurve(
      Vector2Curve=DataInterface() {propertyValues {XCurve {keys [<at 0 1 RCIM_Cubic>,<at 1 0 RCIM_Cubic>]}, YCurve {keys [<at 0 1 RCIM_Cubic>,<at 1 0 RCIM_Cubic>]}}}
      ScaleCurve=<0.5 5>
     )
     Spawn Group Mask=0
    )
    Scale Sprite Size Mode=::NewEnumerator2
   )
   &Drag(
    Drag=&MaskFloatBySpawnGroup(Masked Float=0, Passthrough Float=&RandomRangeFloat(Minimum=0.8, Maximum=1.2))
    Use Linear Drag=true
    Use Rotational Drag=true
   )
   &SolveForcesAndVelocity(Clamp Velocity=false, Limit Acceleration=false)
   &DynamicMaterialParameters(
    Index 0 Param 1=&FloatFromCurve(FloatCurve=DataInterface() {propertyValues {Curve {keys [
         <at 0 1>
         {arriveTangent=-2.1933, interpMode=RCIM_Cubic, leaveTangent=-2.1933, tangentMode=RCTM_User, time=0.2, value=1}
         {arriveTangent=-1.6325, interpMode=RCIM_Cubic, leaveTangent=-1.6325, tangentMode=RCTM_User, time=1, value=0}
        ]}}})
    Module.Index0Param0Write=true
    Module.Index0Param1Write=true
    Module.Index0Param2Write=true
    Module.Index0Param3Write=true
   )
   &ScaleColor(
    Scale Alpha=&MaskFloatBySpawnGroup(
     Masked Float=1
     Passthrough Float=&FloatFromCurve(FloatCurve=DataInterface() {propertyValues {Curve {
         defaultValue=1
         keys [
          {tangentMode=RCTM_User, time=0, value=1}
          {arriveTangent=-0.0405, interpMode=RCIM_Cubic, leaveTangent=-0.0405, tangentMode=RCTM_User, time=0.7009, value=1}
          {arriveTangent=-8.0089, interpMode=RCIM_Cubic, leaveTangent=-8.0089, tangentMode=RCTM_User, time=1, value=0}
         ]
        }}})
    )
    Scale RGB=&MaskVectorBySpawnGroup(
     Masked Vector=&MakeVectorFromLinearColorRGB(Color=&LerpLinearColors(StartColor=<rgba 1 1 1 5>, EndColor=<rgba 0.7 0.7 0.7 1>, LerpFactor=$User.Variant))
     Passthrough Vector=<1 1 1>
    )
    ScaleA=true
    ScaleRGB=true
   )
   &FN_Sprite_Attributes(
    Color=&LerpLinearColors(StartColor=<rgba 1 1 1 1>, EndColor=<rgba 0.01 0.01 0.01 3>, LerpFactor=$User.Variant)
    Create Color Attribute=true
    Create SpriteSize Attribute=true
    SpriteSize=&Multiply_Vector2D(
     A=<40 80>
     B=&Vector2DFromCurve(
      Vector2Curve=DataInterface() {propertyValues {
        XCurve {keys [
          {arriveTangent=0.0233, interpMode=RCIM_Cubic, leaveTangent=0.0233, tangentMode=RCTM_User, time=0, value=1}
          {tangentMode=RCTM_User, time=1, value=0.8}
         ]}
        YCurve {keys [
          {arriveTangent=1.992, interpMode=RCIM_Cubic, leaveTangent=1.992, tangentMode=RCTM_User, time=0, value=1}
          {arriveTangent=0.5299, tangentMode=RCTM_User, time=1, value=2.25}
         ]}
       }}
      ScaleCurve=&Multiply_Vector2D_ByFloat(Float=1, Vector2D=&MakeVector2D(X=1, Y=0.8))
     )
    )
    Write to Module Namespace=true
   )
  ]
  rendererClasses [&NiagaraSpriteRendererProperties]
  renderers [
   {rendererClass=&NiagaraSpriteRendererProperties}
   {rendererClass=&NiagaraSpriteRendererProperties, rendererIndex=1}
   {rendererClass=&NiagaraSpriteRendererProperties, rendererIndex=2}
  ]
 }
 Empty {rendererClasses [], renderers []}
 NE_FN_PhysMatDebris {
  bEnabled=false
  emitterUpdateScript [
   &EmitterState()
   &SpawnBurst_Instantaneous(
    Spawn Count=&RandomRangeInt(Minimum=5, Maximum=10)
    Use Spawn Probability=false
    UseLoopCountLimit=false
   )
  ]
  particleSpawnScript [
   /SetVariables(Particles.BiasedRandom=&FloatFromCurve(
     FloatCurve=DataInterface() {propertyValues {Curve {keys [
         <at 0.2 1 RCIM_Cubic>
         {arriveTangent=-3.3394, interpMode=RCIM_Cubic, leaveTangent=-3.3394, tangentMode=RCTM_User, time=0.5, value=0.5}
         <at 0.8 0 RCIM_Cubic>
        ]}}}
     CurveIndex=&RandomRangeFloat(Minimum=0, Maximum=1)
    )) {isSetParameters=true}
   &InitializeParticle(
    Color Mode=::NewEnumerator0
    Lifetime=&RandomRangeFloat(Minimum=0.5, Maximum=1.25)
    Mass=&Lerp_Float(A=0.6, B=1.2, Alpha=$Particles.BiasedRandom)
    Mass Mode=::NewEnumerator1
    Mesh Scale=&Multiply_VectorByFloat(
     Vector=&RandomRangeVector(Minimum=<0.9 0.4 0.5>, Maximum=<1 1 1>)
     Float=&Lerp_Float(A=0.7, B=1, Alpha=$Particles.BiasedRandom)
    )
    Mesh Scale Mode=::NewEnumerator4
    UsePositionOffset=false
    Write Material Random=true
   )
   &FN_PhysicalSurface(
    Dirt Mesh Count=2
    Glass Mesh Count=3
    Ice Mesh Count=2
    Metal Gravity=&RandomRangeFloat(Minimum=-350, Maximum=-1800)
    Metal Mesh Count=2
    Metal Velocity=&RandomRangeFloat(Minimum=80, Maximum=375)
    Metal Velocity Cone Angle=130
    Stone Gravity=&RandomRangeFloat(Minimum=-700, Maximum=-1250)
    Stone Mesh Count=2
    Stone Velocity=&RandomRangeFloat(Minimum=80, Maximum=550)
    Stone Velocity Cone Angle=120
    Wood Gravity=&RandomRangeFloat(Minimum=-450, Maximum=-900)
    Wood Mesh Count=2
    Wood Scale=1.5
    Wood Velocity=&RandomRangeFloat(Minimum=80, Maximum=600)
    Wood Velocity Cone Angle=130
   )
   &AddVelocityInCone(
    Cone Angle=$Output.FN_PhysicalSurface.SurfaceResponseConeAngle
    Module.Use Velocity Falloff On Cone Axis=false
    Velocity Strength=$Output.FN_PhysicalSurface.SurfaceResponseVelocity
   )
   &InitialMeshOrientation(Use Rotation Vector=false)
   &AddRotationalVelocity(Rotation Rate=&RandomVector(Vector Scale=&RandomRangeFloat(Minimum=1, Maximum=2)))
   &ApplyInitialForces()
   /SetVariables(Particles.GravityOverrideValue=$Output.FN_PhysicalSurface.GravityOverride) {isSetParameters=true}
  ]
  particleUpdateScript [
   &ParticleState()
   &GravityForce(Gravity=$Particles.GravityOverrideValue)
   &Collision(
    Advanced Aging Rate=1
    Age Colliding Particles=false
    CPU Collision Trace Channel=ECC_WorldStatic
    Collision Enabled=&Boolean_AndOperation(
     Bool1=&SetBoolByFloatComparison(A=$Particles.Initial.Lifetime, B=1)
     Bool2=&SetBoolByFloatComparison(A=$Particles.Age, B=0.12)
    )
    Control Roll On Collision=true
    Friction=1
    Friction During a Bounce=0.25
    Max CPU Trace Length=100
    Module.Control Rest State Via Surface Normal=false
    Module.EnableMaxCollisionCount=false
    Module.Randomize Collision Normal Vector=true
    Randomize Collision Normal=0.25
    ScaleRateOfRotation=false
   ) {enabled=false}
   &SolveForcesAndVelocity(Clamp Velocity=false, Limit Acceleration=false)
   &ScaleMeshSize(Scale Factor=&VectorFromFloat(Value=&FloatFromCurve(FloatCurve=DataInterface() {propertyValues {Curve {keys [<at 0.6 1 RCIM_Cubic>, {arriveTangent=-6.308, interpMode=RCIM_Cubic, leaveTangent=-6.308, tangentMode=RCTM_User, time=1, value=0}]}}})))
   &DynamicMaterialParameters(Module.Param0WriteEnabled=true, Module.Param1WriteEnabled=true, Module.Param2WriteEnabled=false, Module.Param3WriteEnabled=false)
  ]
  rendererClasses [&NiagaraSpriteRendererProperties, &NiagaraMeshRendererProperties]
  renderers [
   {rendererClass=&NiagaraSpriteRendererProperties}
   {rendererClass=&NiagaraMeshRendererProperties, rendererIndex=1}
   {rendererClass=&NiagaraMeshRendererProperties, rendererIndex=2}
   {rendererClass=&NiagaraMeshRendererProperties, rendererIndex=3}
   {rendererClass=&NiagaraMeshRendererProperties, rendererIndex=4}
   {rendererClass=&NiagaraMeshRendererProperties, rendererIndex=5}
   {rendererClass=&NiagaraMeshRendererProperties, rendererIndex=6}
   {rendererClass=&NiagaraMeshRendererProperties, rendererIndex=7}
   {rendererClass=&NiagaraMeshRendererProperties, rendererIndex=8}
  ]
 }
 SmokeBlackDetail {
  emitterUpdateScript [
   &EmitterState(Life Cycle Mode=::NewEnumerator1, Loop Behavior=::NewEnumerator1, Loop Delay=0.4, Loop Duration=1, UseLoopDelay=false)
   &SpawnBurst_Instantaneous(
    Spawn Count=&MakeCustomIntFromBool(
     Bool=&SetBoolByFloatComparison(A=$User.Variant, B=1, Comparison Type=::NewEnumerator4)
     True INT=&RandomRangeInt(Minimum=6, Maximum=8)
     False INT=0
    )
    Use Spawn Probability=false
    UseLoopCountLimit=false
   )
   &SpawnBurst_Instantaneous(
    Spawn Count=&MakeCustomIntFromBool(
     Bool=&SetBoolByFloatComparison(A=$User.Variant, B=1, Comparison Type=::NewEnumerator4)
     True INT=&RandomRangeInt(Minimum=4, Maximum=6)
     False INT=0
    )
    Spawn Group=1
    Use Spawn Probability=false
    UseLoopCountLimit=false
   ) {module=SpawnBurst_Instantaneous001}
  ]
  particleSpawnScript [
   /SetVariables(Particles.VisibilityTag=&SpawnGroup()) {isSetParameters=true}
   &InitializeParticle(
    Color=<rgba 1 1 1 2>
    Lifetime=&MaskFloatBySpawnGroup(
     Masked Float=&RandomRangeFloat(Minimum=0.325, Maximum=0.6)
     Passthrough Float=&RandomRangeFloat(Minimum=0.3, Maximum=0.7)
    )
    Position Offset=<0 0 20>
    Sprite Rotation Mode=::NewEnumerator2
    Sprite Size Mode=::NewEnumerator3
    Uniform Sprite Size=&Multiply_Float(
     A=$System.FEScale
     B=&MaskFloatBySpawnGroup(
      Masked Float=&RandomRangeFloat(Minimum=120, Maximum=170)
      Passthrough Float=&RandomRangeFloat(Minimum=170, Maximum=200)
     )
    )
    UsePositionOffset=false
    Write Material Random=true
   )
   &ShapeLocation(MaskSpawnGroup=false, Override Seed=false, Sphere Radius=8, UseOwnerScale=true)
   &CameraOffset(Camera Offset Amount=30)
   &AddVelocity(
    Module.Use Falloff Distance=false
    Override Seed=false
    Velocity Mode=::NewEnumerator1
    Velocity Speed=&MaskFloatBySpawnGroup(
     Masked Float=&RandomRangeFloat(Minimum=45, Maximum=70)
     Passthrough Float=&RandomRangeFloat(Minimum=25, Maximum=40)
    )
   )
  ]
  particleUpdateScript [
   &ParticleState()
   &SpriteRotationRate(Rotation Rate=&MaskFloatBySpawnGroup(
     Masked Float=&RandomRangeFloat(Minimum=-50, Maximum=50)
     Passthrough Float=&RandomRangeFloat(Minimum=-200, Maximum=200)
    ))
   &ScaleSpriteSize(
    Scale Sprite Size Mode=::NewEnumerator0
    Uniform Scale Factor=&FloatFromCurve(FloatCurve=DataInterface() {propertyValues {Curve {keys [<at 0 0.8>,<at 1 1>]}}}, Scale Curve=0.45)
   )
   &Drag(Drag=&RandomRangeFloat(Minimum=0.8, Maximum=1.2), Use Linear Drag=true, Use Rotational Drag=true)
   &SubUVAnimation(
    End Frame Range Override=30
    "Sprite Renderer "=DataInterface() {propertyValues {SpriteRenderer=&self:NiagaraEmitter_3.NiagaraSpriteRendererProperties_0}}
    Start Frame Range Override=&RandomRangeInt(Minimum=0, Maximum=5)
    UseEndFrame=true
    UseStartFrame=true
   )
   &SubUVAnimation(
    End Frame Range Override=32
    Evaluation Type=::NewEnumerator0
    "Sprite Renderer "=DataInterface() {propertyValues {SpriteRenderer=&self:NiagaraEmitter_3.NiagaraSpriteRendererProperties_2}}
    Start Frame Range Override=&RandomRangeInt(Minimum=0, Maximum=12)
    SubUV Animation Mode:ENiagara_SubUVLookupModeV2=::NewEnumerator2
    UseEndFrame=true
    UseStartFrame=true
   ) {module=SubUVAnimation001}
   &ScaleColor(
    Scale Alpha=&Multiply_Float(
     A=&MaskFloatBySpawnGroup(
      Masked Float=&RandomRangeFloat(Minimum=0.5, Maximum=1)
      Passthrough Float=&RandomRangeFloat(Minimum=0.2, Maximum=0.325)
     )
     B=&FloatFromCurve(
      FloatCurve=DataInterface() {propertyValues {Curve {keys [
          {interpMode=RCIM_Cubic, tangentMode=RCTM_User, time=0, value=0}
          {arriveTangent=0.579, interpMode=RCIM_Cubic, leaveTangent=0.579, tangentMode=RCTM_User, time=0.35, value=0.85}
          {interpMode=RCIM_Cubic, tangentMode=RCTM_User, time=0.4997, value=0.9}
          {tangentMode=RCTM_User, time=1, value=0}
         ]}}}
      Scale Curve=&Multiply_Float(A=0.7, B=&FNTimeOfDayScalar(Night Scale=0.3, Morning Scale=0.6, Evening Scale=0.7))
     )
    )
    Scale RGB=&Multiply_VectorByFloat(
     Vector=&MakeVectorFromLinearColorRGB(Color=&ColorFromCurve(ColorCurve=DataInterface() {propertyValues {
         AlphaCurve {keys []}
         BlueCurve {keys [<at 0.2544 0 RCIM_Cubic>,<at 0.9908 0.0365 RCIM_Cubic>]}
         GreenCurve {keys [<at 0.2544 0 RCIM_Cubic>,<at 0.9908 0.0365 RCIM_Cubic>]}
         RedCurve {keys [<at 0.2544 0 RCIM_Cubic>,<at 0.9908 0.0365 RCIM_Cubic>]}
        }}))
     Float=&RandomRangeFloat(Minimum=1, Maximum=4)
    )
    ScaleA=true
    ScaleRGB=true
   )
   &SolveForcesAndVelocity(Clamp Velocity=false, Limit Acceleration=false)
  ]
  rendererClasses [&NiagaraSpriteRendererProperties]
  renderers [
   {rendererClass=&NiagaraSpriteRendererProperties}
   {rendererClass=&NiagaraSpriteRendererProperties, rendererIndex=1}
  ]
 }
 SmokeEnergy {
  emitterUpdateScript [
   &EmitterState(Life Cycle Mode=::NewEnumerator1, Loop Behavior=::NewEnumerator1, Loop Duration=1, UseLoopDelay=false)
   &SpawnBurst_Instantaneous(
    Spawn Count=8
    Spawn Probability=&MakeCustomFloatFromBool(Bool=$User.bIsFrontEndPreview, True Float=0, False Float=1)
    Use Spawn Probability=true
   )
  ]
  particleSpawnScript [
   &PartitionParticles()
   &InitializeParticle(
    Color=&LerpLinearColors(
     StartColor=<rgba 1 1 1 0.4>
     EndColor=&RandomRangeLinearColor(Minimum=<rgba 0.0573 0.0573 0.0573 1>, Maximum=<rgba 0.02 0.02 0.02 3>)
     LerpFactor=$User.Variant
    )
    Lifetime Max=0.5
    Lifetime Min=0.3
    Lifetime Mode=::NewEnumerator1
    Sprite Rotation Mode=::NewEnumerator2
    Sprite Size Mode=::NewEnumerator1
    Uniform Sprite Size Max=230
    Uniform Sprite Size Min=200
    UsePositionOffset=false
    Write Material Random=true
   ) {module=InitializeParticle001}
   &CameraOffset(Camera Offset Amount=&FN_FrontEndFloatModifier(Gameplay=&UniformRangedFloat(Minimum=50, Maximum=60), Front End=0))
   &ShapeLocation(MaskSpawnGroup=false, Override Seed=false, Sphere Radius=10, UseOwnerScale=true)
   &AddVelocity(Velocity=&MakeCustomVectorFromBool(
     Bool=false
     True Vector=<-100 -50 200>
     False Vector=&Multiply_Vector(
      A=&MakeCustomVectorFromBool(
       Bool=false
       True Vector=&SelectVectorFromArray(
        Direct Array Index=$Particles.PartitionParticles.Partition
        Vector Selection Array=DataInterface() {propertyValues {FloatData [<-200 0 -50>, <200 0 -50>]}}
       )
       False Vector=<250 0 -50>
      )
      B=&RandomRangeVector(Minimum=<1 0 0.6>, Maximum=<2 0 1>)
     )
    )) {enabled=false}
   &AddVelocity(
    Cone Axis=&Lerp_Vector(
     A=$System.DirectionalVector
     B=&SelectVectorFromArray(
      Direct Array Index=$Particles.PartitionParticles.Partition
      Vector Selection Array=DataInterface() {propertyValues {FloatData [<-1 0 -0.7>, <1 0 -0.7>]}}
     )
     Alpha=$System.isLastSwing
    )
    Module.UseConeFalloff=false
    Module.UseVelDistribution=false
    Override Seed=false
    Velocity Mode=::NewEnumerator2
    Velocity Speed=&RandomRangeFloat(Minimum=280, Maximum=625)
   ) {module=AddVelocity001}
  ]
  particleUpdateScript [
   &ParticleState()
   &Drag(Drag=5, Use Linear Drag=true, Use Rotational Drag=true)
   &SpriteRotationRate(Rotation Rate=&Multiply_Float(
     A=150
     B=&FloatFromCurve(
      FloatCurve=DataInterface() {propertyValues {Curve {keys [<at 0 1>,<at 1 0>]}}}
      Scale Curve=&RandomRangeFloat(Minimum=-1, Maximum=1)
     )
    ))
   &SolveForcesAndVelocity(Clamp Velocity=false, Limit Acceleration=false)
   &ScaleSpriteSize(Scale Factor=&Vector2DFromFloat(Value=&FloatFromCurve(
      FloatCurve=DataInterface() {propertyValues {Curve {keys [
          {arriveTangent=0.7575, interpMode=RCIM_Cubic, leaveTangent=0.7575, tangentMode=RCTM_User, time=0, value=0.2}
          {arriveTangent=0.2277, interpMode=RCIM_Cubic, leaveTangent=0.2277, tangentMode=RCTM_User, time=1, value=0.75}
         ]}}}
      Scale Curve=&MakeCustomFloatFromBool(Bool=$User.bIsFrontEndPreview, True Float=0.6, False Float=1)
     )))
   &ScaleColor(
    Scale Alpha=&Multiply_Float(
     A=&FloatFromCurve(FloatCurve=DataInterface() {propertyValues {Curve {keys [
          {interpMode=RCIM_Cubic, tangentMode=RCTM_User, time=0, value=0}
          {arriveTangent=-2.3416, interpMode=RCIM_Cubic, leaveTangent=-2.3416, tangentMode=RCTM_User, time=0.2, value=1}
          {arriveTangent=-1.25, interpMode=RCIM_Cubic, leaveTangent=-1.25, time=0.5, value=0.35}
          {tangentMode=RCTM_User, time=1, value=0}
         ]}}})
     B=1
    )
    ScaleA=true
    ScaleRGB=true
   )
  ]
  rendererClasses [&NiagaraSpriteRendererProperties]
  renderers [{rendererClass=&NiagaraSpriteRendererProperties}]
 }
 glow {
  emitterUpdateScript [
   &EmitterState(Life Cycle Mode=::NewEnumerator1, Loop Behavior=::NewEnumerator1, Loop Duration=0.9, UseLoopDelay=false)
   &SpawnBurst_Instantaneous(Use Spawn Probability=false, UseLoopCountLimit=false)
   &SpawnBurst_Instantaneous(Spawn Group=1, Use Spawn Probability=false, UseLoopCountLimit=false) {module=SpawnBurst_Instantaneous001}
  ]
  particleSpawnScript [
   /SetVariables(Particles.VisibilityTag=&SpawnGroup()) {isSetParameters=true}
   &InitializeParticle(
    Color=&MaskLinearColorBySpawnGroup(
     Masked Linear Color=&LerpLinearColors(StartColor=<rgba 0.401 0.401 0.401 1.2>, EndColor=<rgba 0.15 0.15 0.15 1.5>, LerpFactor=$User.Variant)
     Passthrough Linear Color=&LerpLinearColors(StartColor=<rgba 5 5 5 1>, EndColor=<rgba 1 1 1 1>, LerpFactor=$User.Variant)
    )
    Lifetime=&MaskFloatBySpawnGroup(Masked Float=0.4, Passthrough Float=0.16)
    Position Offset=&MakeCustomVectorFromBool(Bool=false, True Vector=<-10 0 0>, False Vector=<0 0 0>)
    Sprite Rotation Angle=&Lerp_Float(A=0, B=&RandomRangeFloat(Minimum=0, Maximum=360), Alpha=$User.Variant)
    Sprite Rotation Mode=::NewEnumerator0
    Sprite Size=&MaskVector2dBySpawnGroup(
     Masked Vector2D=&Multiply_Vector2D_ByFloat(Float=$System.FEScale, Vector2D=<120 120>)
     Passthrough Vector2D=&Lerp_Vector2(A=<120 120>, B=<100 100>, Alpha=$User.Variant)
     Spawn Group Mask=0
    )
    Sprite Size Mode=::NewEnumerator4
    UsePositionOffset=true
    Write Material Random=true
   )
   &CameraOffset(Camera Offset Amount=50)
   /SetVariables(
    Particles.FlashSubImageIndex:NiagaraFloat=&RandomRangeFloat(Minimum=0, Maximum=3)
    Particles.GlowCameraOffset=15
   ) {isSetParameters=true}
  ]
  particleUpdateScript [
   &ParticleState()
   &ScaleSpriteSize(
    Scale Sprite Size Mode=::NewEnumerator0
    Uniform Scale Factor=&MaskFloatBySpawnGroup(
     Masked Float=&FloatFromCurve(
      FloatCurve=DataInterface() {propertyValues {Curve {keys [<at 0 1.3>, {tangentMode=RCTM_User, time=1, value=0.5}]}}}
      Scale Curve=&MakeCustomFloatFromBool(Bool=$User.bIsFrontEndPreview, True Float=0.6, False Float=1)
     )
     Passthrough Float=&FloatFromCurve(FloatCurve=DataInterface() {propertyValues {Curve {
         defaultValue=0
         keys [
          {arriveTangent=9.0725, interpMode=RCIM_Cubic, leaveTangent=9.0725, tangentMode=RCTM_User, time=0, value=0}
          {arriveTangent=-0.1476, interpMode=RCIM_Cubic, leaveTangent=-0.1476, tangentMode=RCTM_User, time=0.1408, value=0.9847}
          {arriveTangent=-1.7969, interpMode=RCIM_Cubic, leaveTangent=-1.7969, tangentMode=RCTM_User, time=1, value=0}
         ]
        }}})
    )
   )
   /SetVariables(Particles.Size002:Vector2f=&Multiply_Vector2D(
     A=$Particles.SpriteSize
     B=&MakeCustomVector2DFromBool(Bool=$User.bIsFrontEndPreview, True Vector2D=<0.95 0.75>, False Vector2D=<2 1.6>)
    )) {isSetParameters=true}
   &ScaleColor(
    Scale Alpha=&MaskFloatBySpawnGroup(
     Masked Float=&FloatFromCurve(
      FloatCurve=DataInterface() {propertyValues {Curve {keys [
          {arriveTangent=4.4475, interpMode=RCIM_Cubic, leaveTangent=4.4475, tangentMode=RCTM_User, time=0, value=0.6}
          {arriveTangent=5.3518, interpMode=RCIM_Cubic, leaveTangent=-3.9765, tangentMode=RCTM_Break, time=0.12, value=1.3}
          {arriveTangent=-1.9582, interpMode=RCIM_Cubic, leaveTangent=-1.9582, tangentMode=RCTM_User, time=0.35, value=0.5}
          <at 1 0 RCIM_Cubic>
         ]}}}
      Scale Curve=&MakeCustomFloatFromBool(Bool=$User.bIsFrontEndPreview, True Float=0.6, False Float=1)
     )
     Passthrough Float=&FloatFromCurve(
      FloatCurve=DataInterface() {propertyValues {Curve {keys [<at 0 1>,<at 0.85 1>,<at 1 0>]}}}
      Scale Curve=&Lerp_Float(A=1, B=0.6, Alpha=$User.Variant)
     )
    )
    Scale RGB=&MaskVectorBySpawnGroup(
     Masked Vector=&Lerp_Vector(
      A=&MakeVectorFromLinearColorRGB(Color=&ColorFromCurve(ColorCurve=DataInterface() {propertyValues {
          AlphaCurve {keys []}
          BlueCurve {keys [<at 0.0182 1 RCIM_Cubic>,<at 0.2369 0.6823 RCIM_Cubic>]}
          GreenCurve {keys [<at 0.0182 1 RCIM_Cubic>,<at 0.2369 0.6823 RCIM_Cubic>]}
          RedCurve {keys [<at 0.0182 1 RCIM_Cubic>,<at 0.2369 0.6823 RCIM_Cubic>]}
         }}))
      B=&MakeVectorFromLinearColorRGB(Color=&ColorFromCurve(ColorCurve=DataInterface() {propertyValues {
          AlphaCurve {keys []}
          BlueCurve {keys [<at 0.0182 0.5 RCIM_Cubic>,<at 0.2369 0 RCIM_Cubic>]}
          GreenCurve {keys [<at 0.0182 0.5 RCIM_Cubic>,<at 0.2369 0 RCIM_Cubic>]}
          RedCurve {keys [<at 0.0182 0.5 RCIM_Cubic>,<at 0.2369 0 RCIM_Cubic>]}
         }}))
      Alpha=$User.Variant
     )
     Passthrough Vector=&MakeVectorFromLinearColorRGB(Color=&ColorFromCurve(ColorCurve=DataInterface() {propertyValues {
         AlphaCurve {keys []}
         BlueCurve {keys [<at 0.0182 3 RCIM_Cubic>, {arriveTangent=-3.0847, interpMode=RCIM_Cubic, leaveTangent=-3.0847, time=0.3657, value=1}, <at 0.9908 0>]}
         GreenCurve {keys [<at 0.0182 3 RCIM_Cubic>, {arriveTangent=-3.0847, interpMode=RCIM_Cubic, leaveTangent=-3.0847, time=0.3657, value=1}, <at 0.9908 0>]}
         RedCurve {keys [<at 0.0182 3 RCIM_Cubic>, {arriveTangent=-3.0847, interpMode=RCIM_Cubic, leaveTangent=-3.0847, time=0.3657, value=1}, <at 0.9908 0>]}
        }}))
    )
    ScaleA=true
    ScaleRGB=true
   )
   /SetVariables(Particles.color002:LinearColor=&MakeLinearColorFromVectorAndFloat(
     "Vector (RGB)"=&Multiply_Vector(
      A=&MakeVectorFromLinearColorRGB(Color=&LerpLinearColors(StartColor=<rgba 2 2 2 1>, EndColor=<rgba 1 1 1 1>, LerpFactor=$User.Variant))
      B=&MakeVectorFromLinearColorRGB(Color=&FNTimeOfDayColor(Night Color=<rgba 0.2 0.35 0.8 1>, Morning Color=<rgba 0.8 0.8 0.8 1>, Day Color=<rgba 0.8 0.8 0.8 1>, Evening Color=<rgba 0.8 0.8 0.8 1>))
     )
     "Float (Alpha)"=&FloatFromCurve(FloatCurve=DataInterface() {propertyValues {Curve {keys [
          {arriveTangent=-4.1686, interpMode=RCIM_Cubic, leaveTangent=-4.1686, tangentMode=RCTM_User, time=0, value=1}
          {arriveTangent=-1.5469, interpMode=RCIM_Cubic, leaveTangent=-1.5469, tangentMode=RCTM_User, time=0.25, value=0}
         ]}}})
    )) {isSetParameters=true}
  ]
  rendererClasses [&NiagaraSpriteRendererProperties]
  renderers [
   {rendererClass=&NiagaraSpriteRendererProperties}
   {rendererClass=&NiagaraSpriteRendererProperties, rendererIndex=1}
   {rendererClass=&NiagaraSpriteRendererProperties, rendererIndex=2}
  ]
 }
}
summary {
 systemName=NS_Pickaxe_StuffRasp_Maze_Impact
 emitters [
  {
   bEnabled=false
   emitterName=NE_FN_PhysMatDebris
   rendererClasses [&NiagaraSpriteRendererProperties, &NiagaraMeshRendererProperties]
  }
  {emitterName=glow, rendererClasses [&NiagaraSpriteRendererProperties]}
  {emitterName=SmokeEnergy, rendererClasses [&NiagaraSpriteRendererProperties]}
  {emitterName=Claw, rendererClasses [&NiagaraSpriteRendererProperties]}
  {emitterName=Center_spikes, rendererClasses [&NiagaraSpriteRendererProperties]}
  {emitterName=SmokeBlackDetail, rendererClasses [&NiagaraSpriteRendererProperties]}
  {emitterName=Empty, rendererClasses []}
 ]
 userVariables [
  User.bIsFrontEnd {type=NiagaraBool}
  User.bIsFrontEndPreview {type=NiagaraBool}
  User.DEBUG {type=NiagaraFloat, defaultValue=0, description=""}
  User.SwingSectionIndex {type=NiagaraInt32}
  User.Variant {type=NiagaraFloat, defaultValue=0, description=""}
 ]
}
