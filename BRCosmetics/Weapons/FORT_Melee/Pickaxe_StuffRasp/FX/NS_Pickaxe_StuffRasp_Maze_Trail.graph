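# Niagara swing trail effect for the StuffRasp "Maze" pickaxe, with claw and smoke ribbons plus ember and smoke sprites.
# Claw_Trail builds four anim-trail ribbons from bone or socket pairs, and Smoke_Trail layers two more on a further pair.
# Embers and Smoke spawn by distance moved at skeletal mesh samples, using curl noise and flipbook animation.
# User.Variant blends claw color and fade and raises smoke and ember alpha from zero, with time-of-day adjustments.
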NiagaraSystem /BRCosmetics/Weapons/FORT_Melee/Pickaxe_StuffRasp/FX/NS_Pickaxe_StuffRasp_Maze_Trail
dependencies {
 usedDataInterfaces [
  &NiagaraDataInterfaceCurve
  &NiagaraDataInterfaceColorCurve
  &NiagaraDataInterfaceVector2DCurve
  &NiagaraDataInterfaceSpriteRendererInfo
  &NiagaraDataInterfaceSocketReader
  &NiagaraDataInterfaceSkeletalMesh
 ]
 usedDynamicInputs [&Vector2DFromCurve, &RandomRangeVector2D]
 usedModules [
  &self:SystemSpawnScript.SystemScriptSource.SystemScriptGraph.NiagaraNodeAssignment_1.SetVariables_#1
  &SystemState
 ]
}
emitters {
 Claw_Trail {
  emitterSpawnScript [&FN_RibbonIDOffsetUtility()]
  emitterUpdateScript [
   &EmitterState()
   &FN_SpawnAnimTrail(Number of Anim Trails=4)
   /SetVariables(
    Emitter.Color A:LinearColor=&LerpLinearColors(StartColor=<rgba 0.2371 0.2935 0.4844 1>, EndColor=<rgba 0.2016 0.2016 0.2016 1>, LerpFactor=$User.Variant)
    Emitter.Color B:LinearColor=&LerpLinearColors(StartColor=<rgba 0.5668 0.6415 0.8308 1>, EndColor=<rgba 0.4678 0.4678 0.4678 1>, LerpFactor=$User.Variant)
   ) {isSetParameters=true}
  ]
  particleSpawnScript [
   &FN_InitializeAnimTrailBoneIndecies()
   &FN_InitializeAnimTrailBoneIndecies(AnimTrail Selection=1, First Bone or Socket Index=2, Second Bone or Socket Index=3) {module=FN_InitializeAnimTrailBoneIndecies001}
   &FN_InitializeAnimTrailBoneIndecies(AnimTrail Selection=2, First Bone or Socket Index=4, Second Bone or Socket Index=5) {module=FN_InitializeAnimTrailBoneIndecies002}
   &FN_InitializeAnimTrailBoneIndecies(AnimTrail Selection=3, First Bone or Socket Index=6, Second Bone or Socket Index=7) {module=FN_InitializeAnimTrailBoneIndecies003}
   &FN_InitializeAnimTrail(
    AnimTrailWidth=1.45
    Color=&MergeHelper_ColorSwitch(
     0=<rgba 1 1 1 1>
     1=<rgba 0.7656 0.7656 0.7656 1>
     2=<rgba 0.6042 0.6042 0.6042 1>
     3=<rgba 0.6042 0.6042 0.6042 1>
     Index=&SpawnGroup()
     Num=::NewEnumerator2
    )
    Lifetime=0.325
   )
   &AddVelocity(Velocity=<0 400 0>) {enabled=false}
  ]
  particleUpdateScript [
   &SolveForcesAndVelocity(Clamp Velocity=false, Limit Acceleration=false)
   &FN_UpdateAnimTrail() {module=FN_UpdateAnimTrail001}
   &ScaleRibbonWidth(Ribbon Width Scale=&FloatFromCurve(FloatCurve=DataInterface() {propertyValues {Curve {keys [{arriveTangent=-0.0042, interpMode=RCIM_Cubic, leaveTangent=-0.0042, tangentMode=RCTM_User, time=0, value=1}, <at 1 0>]}}}))
   &ScaleColor(
    Scale Alpha=&FloatFromCurve(
     FloatCurve=DataInterface() {propertyValues {Curve {keys [
         {arriveTangent=-1.3965, interpMode=RCIM_Cubic, leaveTangent=-1.3965, tangentMode=RCTM_User, time=0, value=1}
         {interpMode=RCIM_Cubic, tangentMode=RCTM_User, time=1, value=0}
        ]}}}
     Scale Curve=&Lerp_Float(
      A=&Multiply_Float(
       A=&MergeHelper_FloatSwitch(0=1, 1=0.6, 2=0.525, 3=0.45, Index=&SpawnGroup(), Num=::NewEnumerator2)
       B=&FNTimeOfDayScalar(Night Scale=0.6, Morning Scale=1, Evening Scale=1)
      )
      B=&MergeHelper_FloatSwitch(0=1, 1=0.7, 2=0.5, 3=0.5, Index=&SpawnGroup(), Num=::NewEnumerator2)
      Alpha=$User.Variant
     )
    )
    Scale RGB=&Lerp_Vector(
     A=&Multiply_Vector(
      A=&MakeVectorFromLinearColorRGB(Color=&MaskLinearColorBySpawnGroup(
        Masked Linear Color=&ColorFromCurve(ColorCurve=DataInterface() {propertyValues {
           AlphaCurve {keys []}
           BlueCurve {keys [<at 0 2>, <at 0.325 1>, {arriveTangent=-4.3928, leaveTangent=-4.3928, time=1, value=0.1391}]}
           GreenCurve {keys [<at 0 2>, <at 0.325 1>, {arriveTangent=-4.3928, leaveTangent=-4.3928, time=1, value=0.1391}]}
           RedCurve {keys [<at 0 2>, <at 0.325 1>, {arriveTangent=-4.3928, leaveTangent=-4.3928, time=1, value=0.1391}]}
          }})
        Passthrough Linear Color=&ColorFromCurve(ColorCurve=DataInterface() {propertyValues {
           AlphaCurve {keys []}
           BlueCurve {keys [<at 0 2>, <at 0.325 1>, {arriveTangent=-4.3928, leaveTangent=-4.3928, time=1, value=1}]}
           GreenCurve {keys [<at 0 2>, <at 0.325 1>, {arriveTangent=-4.3928, leaveTangent=-4.3928, time=1, value=1}]}
           RedCurve {keys [<at 0 2>, <at 0.325 1>, {arriveTangent=-4.3928, leaveTangent=-4.3928, time=1, value=1}]}
          }})
       ))
      B=&MakeVectorFromLinearColorRGB(Color=&FNTimeOfDayColor(Night Color=<rgba 0.4875 0.65 0.7448 1>, Morning Color=<rgba 1 1 1 1>, Day Color=<rgba 1 1 1 1>, Evening Color=<rgba 1 1 1 1>))
     )
     B=&MakeVectorFromLinearColorRGB(Color=&ColorFromCurve(ColorCurve=DataInterface() {propertyValues {
         AlphaCurve {keys []}
         BlueCurve {keys [<at 0.008 1>, {arriveTangent=-4.3928, leaveTangent=-4.3928, time=0.5321, value=0.3}, <at 0.6889 0>]}
         GreenCurve {keys [<at 0.008 1>, {arriveTangent=-4.3928, leaveTangent=-4.3928, time=0.5321, value=0.3}, <at 0.6889 0>]}
         RedCurve {keys [<at 0.008 1>, {arriveTangent=-4.3928, leaveTangent=-4.3928, time=0.5321, value=0.3}, <at 0.6889 0>]}
        }}))
     Alpha=$User.Variant
    )
    ScaleA=true
    ScaleRGB=true
   )
  ]
  rendererClasses [&NiagaraRibbonRendererProperties]
  renderers [
   {rendererClass=&NiagaraRibbonRendererProperties}
   {rendererClass=&NiagaraRibbonRendererProperties, rendererIndex=1}
  ]
 }
 Embers {
  emitterUpdateScript [
   &EmitterState()
   &SpawnPerUnit(Module.Use Max Movement Threshold=true, Module.Use Movement Tolerance=false, Spawn Spacing=8, Use Spawn Probability=false)
  ]
  particleSpawnScript [
   &InitializeParticle(
    Lifetime Max=0.8
    Lifetime Min=0.3
    Lifetime Mode=::NewEnumerator1
    Sprite Rotation Mode=::NewEnumerator2
    Sprite Size Mode=::NewEnumerator1
    Uniform Sprite Size Max=12
    Uniform Sprite Size Min=3
    UsePositionOffset=false
    Write Material Random=true
   )
   &SkeletalMeshLocation(MaskSpawnGroup=false, Mesh Sampling Type=::NewEnumerator5, Skeletal Mesh=$System.Skeletal Mesh Reader, Socket Sampling Mode=::NewEnumerator5)
   &ShapeLocation(
    MaskSpawnGroup=false
    Override Seed=false
    Shape Origin=$Particles.SkeletalMeshLocation.SampledPosition
    Sphere Radius=12
    UseOwnerScale=true
   )
  ]
  particleUpdateScript [
   &ParticleState()
   &SpriteRotationRate(Rotation Rate=&RandomRangeFloat(Minimum=-100, Maximum=100))
   &ScaleSpriteSize(
    Uniform Curve Scale=1.3
    Uniform Curve Sprite Scale=DataInterface() {propertyValues {Curve {keys [
        {arriveTangent=2.2767, interpMode=RCIM_Cubic, leaveTangent=2.2767, tangentMode=RCTM_User, time=0, value=0.2}
        {arriveTangent=-0.009, interpMode=RCIM_Cubic, leaveTangent=-0.009, tangentMode=RCTM_User, time=1, value=1}
       ]}}}
   )
   &GravityForce(Gravity=&RandomRangeVector(Minimum=<-10 -10 -10>, Maximum=<10 10 10>)) {enabled=false}
   &CurlNoiseForce(
    Bias Noise=false
    Noise Frequency=15
    Noise Strength=&RandomRangeFloat(Minimum=10, Maximum=35)
    Pan Noise=false
   )
   &Drag(Drag=&RandomRangeFloat(Minimum=1, Maximum=3), Use Linear Drag=true, Use Rotational Drag=true)
   &SolveForcesAndVelocity(Clamp Velocity=false, Limit Acceleration=false)
   &ScaleColor(
    Scale Alpha=&FloatFromCurve(
     FloatCurve=DataInterface() {propertyValues {Curve {keys [
         <at 0 0 RCIM_Cubic>
         {arriveTangent=-0.1631, interpMode=RCIM_Cubic, leaveTangent=-0.1631, tangentMode=RCTM_User, time=0.22, value=1}
         {arriveTangent=-2.4879, interpMode=RCIM_Cubic, leaveTangent=-2.4879, tangentMode=RCTM_User, time=1, value=0}
        ]}}}
     Scale Curve=&RandomRangeFloat(Minimum=1.2, Maximum=1.6)
    )
    Scale RGB=&MakeVectorFromLinearColorRGB(Color=&ColorFromCurve(ColorCurve=DataInterface() {propertyValues {
        AlphaCurve {keys []}
        BlueCurve {keys [<at 0 0.1953>,<at 0.1732 0.0417 RCIM_Cubic>,<at 0.4583 0>]}
        GreenCurve {keys [<at 0 0.1953>,<at 0.1732 0.0417 RCIM_Cubic>,<at 0.4583 0>]}
        RedCurve {keys [<at 0 0.1953>,<at 0.1732 0.0417 RCIM_Cubic>,<at 0.4583 0>]}
       }}))
    ScaleA=true
    ScaleRGB=true
   )
   &ScaleColor(
    Scale Alpha=&Lerp_Float(
     A=0
     B=&FNTimeOfDayScalar(Night Scale=0.6, Morning Scale=0.7, Evening Scale=0.7)
     Alpha=$User.Variant
    )
    Scale RGB=&MakeVectorFromLinearColorRGB(Color=&FNTimeOfDayColor(Night Color=<rgba 1 0.9948 0.2694 1>, Morning Color=<rgba 1 1 1 1>, Day Color=<rgba 1 1 1 1>, Evening Color=<rgba 1 1 1 1>))
    ScaleA=true
    ScaleRGB=true
   ) {module=ScaleColor001}
   &SubUVAnimation(
    End Frame Range Override:NiagaraInt32=&RandomRangeInt(Minimum=12, Maximum=16)
    "Sprite Renderer "=DataInterface() {propertyValues {SpriteRenderer=&self:NiagaraEmitter_2.NiagaraSpriteRendererProperties_2}}
    Start Frame Range Override=7
    UseEndFrame=true
    UseStartFrame=true
   )
  ]
  rendererClasses [&NiagaraSpriteRendererProperties]
  renderers [{rendererClass=&NiagaraSpriteRendererProperties}]
 }
 Smoke {
  emitterUpdateScript [
   &EmitterState()
   &SpawnPerUnit(Module.Use Max Movement Threshold=true, Module.Use Movement Tolerance=false, Spawn Spacing=6, Use Spawn Probability=false)
   &SpawnPerUnit(Module.Use Max Movement Threshold=true, Module.Use Movement Tolerance=false, Spawn Group=1, Spawn Spacing=4, Use Spawn Probability=false) {module=SpawnPerUnit001}
  ]
  particleSpawnScript [
   /SetVariables(Particles.VisibilityTag=&SpawnGroup()) {isSetParameters=true}
   &InitializeParticle(
    Lifetime=&MaskFloatBySpawnGroup(
     Masked Float=&RandomRangeFloat(Minimum=0.2, Maximum=0.7)
     Passthrough Float=&RandomRangeFloat(Minimum=0.5, Maximum=0.9)
    )
    Position Offset=<0 0 15>
    Sprite Rotation Mode=::NewEnumerator2
    Sprite Size Mode=::NewEnumerator3
    Uniform Sprite Size=&MaskFloatBySpawnGroup(Masked Float=&RandomRangeFloat(Minimum=18, Maximum=25), Passthrough Float=32)
    UsePositionOffset=true
    Write Material Random=true
   )
   &AddVelocity(Module.UseConeFalloff=false, Module.UseVelDistribution=false, Override Seed=false, Velocity Mode=::NewEnumerator2) {enabled=false}
   &SkeletalMeshLocation(MaskSpawnGroup=false, Mesh Sampling Type=::NewEnumerator5, Skeletal Mesh=$System.Skeletal Mesh Reader, Socket Sampling Mode=::NewEnumerator5)
   &SocketLocation(SamplePosOffset=false, Socket Index=8, Socket Reader=$System.Socket Reader, Socket Sampling Mode=::NewEnumerator3, VelocitySpeedLimit=false) {enabled=false}
   &ShapeLocation(MaskSpawnGroup=false, Override Seed=false, Shape Origin=$Particles.SkeletalMeshLocation.SampledPosition, Sphere Radius=8, UseOwnerScale=true)
  ]
  particleUpdateScript [
   &ParticleState()
   &SpriteRotationRate(Rotation Rate=&RandomRangeFloat(Minimum=-30, Maximum=30))
   &ScaleSpriteSize(
    Scale Sprite Size Mode=::NewEnumerator0
    Uniform Scale Factor=&FloatFromCurve(FloatCurve=DataInterface() {propertyValues {Curve {keys [<at 0 0.8>,<at 1 1.2>]}}})
   )
   &SubUVAnimation(
    End Frame Range Override=30
    "Sprite Renderer "=DataInterface() {propertyValues {SpriteRenderer=&self:Smoke_0.NiagaraSpriteRendererProperties_1}}
    Start Frame Range Override=&RandomRangeInt(Minimum=0, Maximum=5)
    UseEndFrame=true
    UseStartFrame=true
   )
   &SubUVAnimation(
    End Frame Range Override=30
    "Sprite Renderer "=DataInterface() {propertyValues {SpriteRenderer=&self:Smoke_0.NiagaraSpriteRendererProperties_2}}
    Start Frame Range Override=&RandomRangeInt(Minimum=0, Maximum=5)
    UseEndFrame=true
    UseStartFrame=true
   ) {module=SubUVAnimation001}
   &ScaleColor(
    Scale Alpha=&FloatFromCurve(
     FloatCurve=DataInterface() {propertyValues {Curve {keys [
         {interpMode=RCIM_Cubic, tangentMode=RCTM_User, time=0, value=0}
         {arriveTangent=0.0136, interpMode=RCIM_Cubic, leaveTangent=0.0136, tangentMode=RCTM_User, time=0.3, value=1}
         {interpMode=RCIM_Cubic, tangentMode=RCTM_User, time=0.5, value=1}
         {tangentMode=RCTM_User, time=1, value=0}
        ]}}}
     Scale Curve=&Lerp_Float(
      A=0
      B=&Multiply_Float(
       A=&MaskFloatBySpawnGroup(
        Masked Float=&RandomRangeFloat(Minimum=0.5525, Maximum=0.78)
        Passthrough Float=&RandomRangeFloat(Minimum=0.13, Maximum=0.2925)
       )
       B=&FNTimeOfDayScalar(Night Scale=0.5, Morning Scale=0.5, Evening Scale=0.5)
      )
      Alpha=$User.Variant
     )
    )
    Scale RGB=&MaskVectorBySpawnGroup(
     Masked Vector=&Lerp_Vector(
      A=&MakeVectorFromLinearColorRGB(Color=<rgba 0.3229 0.3229 0.3229 1>)
      B=&RandomRangeVector(
       Minimum=&MakeVectorFromLinearColorRGB(Color=<rgba 0 0 0 1>)
       Maximum=&MakeVectorFromLinearColorRGB(Color=<rgba 0.0104 0.0104 0.0104 1>)
      )
      Alpha=&FloatFromCurve(FloatCurve=DataInterface() {propertyValues {Curve {
          defaultValue=1
          keys [<at 0 0 RCIM_Cubic>, {arriveTangent=4.1951, interpMode=RCIM_Cubic, leaveTangent=4.1951, tangentMode=RCTM_User, time=0.3181, value=1}]
         }}})
     )
     Passthrough Vector=&MakeVectorFromLinearColorRGB(Color=&RandomRangeLinearColor(Minimum=<rgba 0.0052 0.0052 0.0052 1>, Maximum=<rgba 0.0469 0.0469 0.0469 1>))
    )
    ScaleA=true
    ScaleRGB=true
   )
   &CurlNoiseForce(Bias Noise=false, Noise Strength=&RandomRangeFloat(Minimum=65, Maximum=100), Pan Noise=false)
   &Drag(Drag=&RandomRangeFloat(Minimum=1, Maximum=3), Use Linear Drag=true, Use Rotational Drag=true)
   &SolveForcesAndVelocity(Clamp Velocity=false, Limit Acceleration=false)
  ]
  rendererClasses [&NiagaraSpriteRendererProperties]
  renderers [
   {rendererClass=&NiagaraSpriteRendererProperties}
   {rendererClass=&NiagaraSpriteRendererProperties, rendererIndex=1}
  ]
 }
 Smoke_Trail {
  emitterSpawnScript [&FN_RibbonIDOffsetUtility()]
  emitterUpdateScript [
   &EmitterState()
   &FN_SpawnAnimTrail(Number of Anim Trails=2)
  ]
  particleSpawnScript [
   &FN_InitializeAnimTrailBoneIndecies(First Bone or Socket Index=8, Second Bone or Socket Index=9)
   &FN_InitializeAnimTrailBoneIndecies(AnimTrail Selection=1, First Bone or Socket Index=8, Second Bone or Socket Index=9) {module=FN_InitializeAnimTrailBoneIndecies001}
   &FN_InitializeAnimTrail(AnimTrailWidth=&MaskFloatBySpawnGroup(Masked Float=8, Passthrough Float=5), Lifetime=0.425)
   &AddVelocity(Velocity=<0 400 0>) {enabled=false}
  ]
  particleUpdateScript [
   &SolveForcesAndVelocity(Clamp Velocity=false, Limit Acceleration=false)
   &FN_UpdateAnimTrail() {module=FN_UpdateAnimTrail001}
   &ScaleRibbonWidth(Ribbon Width Scale=&FloatFromCurve(FloatCurve=DataInterface() {propertyValues {Curve {keys [{arriveTangent=-2.4176, interpMode=RCIM_Cubic, leaveTangent=-2.4176, tangentMode=RCTM_User, time=0, value=1}, <at 1 0>]}}}))
   &ScaleColor(
    Scale Alpha=&FloatFromCurve(
     FloatCurve=DataInterface() {propertyValues {Curve {
        defaultValue=1
        keys [{interpMode=RCIM_Cubic, tangentMode=RCTM_User, time=0, value=1}, {interpMode=RCIM_Cubic, tangentMode=RCTM_User, time=1, value=0}]
       }}}
     Scale Curve=&Lerp_Float(
      A=0
      B=&Multiply_Float(A=0.6, B=&MaskFloatBySpawnGroup(Masked Float=0.7, Passthrough Float=0.325))
      Alpha=$User.Variant
     )
    )
    Scale RGB=&MaskVectorBySpawnGroup(
     Masked Vector=&MakeVectorFromLinearColorRGB(Color=&ColorFromCurve(ColorCurve=DataInterface() {propertyValues {
         AlphaCurve {keys []}
         BlueCurve {keys [{arriveTangent=-4.3928, time=0, value=0.3}, <at 0.298 0>]}
         GreenCurve {keys [{arriveTangent=-4.3928, time=0, value=0.3}, <at 0.298 0>]}
         RedCurve {keys [{arriveTangent=-4.3928, time=0, value=0.3}, <at 0.298 0>]}
        }}))
     Passthrough Vector=&MakeVectorFromLinearColorRGB(Color=&ColorFromCurve(ColorCurve=DataInterface() {propertyValues {
         AlphaCurve {keys []}
         BlueCurve {keys [{arriveTangent=-4.3928, time=0, value=1}, <at 1 0.7083>]}
         GreenCurve {keys [{arriveTangent=-4.3928, time=0, value=1}, <at 1 0.7083>]}
         RedCurve {keys [{arriveTangent=-4.3928, time=0, value=1}, <at 1 0.7083>]}
        }}))
    )
    ScaleA=true
    ScaleRGB=true
   )
   &DynamicMaterialParameters(
    Index 0 Param 1=&FloatFromCurve(FloatCurve=DataInterface() {propertyValues {Curve {defaultValue=1, keys [{tangentMode=RCTM_User, time=0, value=0}, {tangentMode=RCTM_User, time=1, value=1}]}}})
    Index 0 Param 3=1
    Index 0 Param 4=1
    Module.Index0Float0Write=true
    Module.Index0Float1Write=true
    Module.Index0Float2Write=true
    Module.Index0Float3Write=true
   )
  ]
  rendererClasses [&NiagaraRibbonRendererProperties]
  renderers [
   {rendererClass=&NiagaraRibbonRendererProperties}
   {rendererClass=&NiagaraRibbonRendererProperties, rendererIndex=1}
  ]
 }
}
summary {
 systemName=NS_Pickaxe_StuffRasp_Maze_Trail
 emitters [
  {emitterName=Smoke_Trail, rendererClasses [&NiagaraRibbonRendererProperties]}
  {emitterName=Smoke, rendererClasses [&NiagaraSpriteRendererProperties]}
  {emitterName=Embers, rendererClasses [&NiagaraSpriteRendererProperties]}
  {emitterName=Claw_Trail, rendererClasses [&NiagaraRibbonRendererProperties]}
 ]
 userVariables [
  User.bIsFrontEnd {type=NiagaraBool}
  User.bIsFrontEndPreview {type=NiagaraBool}
  User.RibbonIDOffset {type=NiagaraInt32}
  User.SwingSectionIndex {type=NiagaraInt32}
  User.Variant {type=NiagaraFloat, defaultValue=0, description=""}
 ]
}
