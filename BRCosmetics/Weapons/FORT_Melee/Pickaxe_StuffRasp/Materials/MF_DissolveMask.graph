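# Material function for the StuffRasp pickaxe that builds a dissolve mask driven by a radius input along the UV's vertical axis.
# It sums a clean sphere mask, a panning noise-distorted one and a softer smoke-detail band, treating the single-channel coordinate as distance from 1.
# The smoke comes from MFW_FlowingSmoke, borrowed from a glider asset.
# Outputs are the combined mask, the smoke alone, and a thin edge from subtracting two thresholds of the combined mask.
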MaterialFunction /BRCosmetics/Weapons/FORT_Melee/Pickaxe_StuffRasp/Materials/MF_DissolveMask
graph {nodes [
  n1 = ComponentMask(None=NamedRerouteUsage()) {g=true}
  n2 = NamedRerouteUsage()
  n3 = Panner(Coordinate=Multiply(A=NamedRerouteUsage(), B=ScalarParameter() {defaultValue=1.3, group=DissolveMask, parameterName=DissolveMask_UVNoiseTiling})) {speedY=0.2}
  n4 = Add(
   A=@n1
   B=Multiply(
    A=TextureSample(UVs=@n3) {samplerType=SAMPLERTYPE_LinearColor, texture=/Game/Global/Textures/Default/Noise/T_Curie_NoisePack}.R
    B=ScalarParameter() {defaultValue=0.124, group=DissolveMask, parameterName=DissolveMask_UVNoiseStrength}
   )
  )
  n5 = Constant() {r=1}
  n6 = NamedRerouteUsage()
  n7 = Constant() {r=0.5904}
  n8 = NamedRerouteUsage()
  n9 = Multiply(
   A=SphereMask(
    A=@n4
    B=Constant() {r=1}
    Radius=Add(A=@n8) {constB=0.1}
    Hardness=ScalarParameter() {defaultValue=0.2, group=DissolveMask, parameterName=DissolveMask_DetailSmokeHardness}
   )
   B=MaterialFunctionCall!(In_FireSmoke_TAMask, In_FireSmoke_UV=NamedRerouteUsage()) {materialFunction=/BRCosmetics/Gadgets/Assets/VinderTech_GliderChute/Glider_RampCanal/Materials/MFW_FlowingSmoke}.Result
  )
  n10 = Add(
   A=SphereMask(A=@n1, B=Constant() {r=1}, Radius=Subtract(A=@n2) {constB=0.1}, Hardness=Constant() {r=0.8})
   B=Add(A=SphereMask(A=@n4, B=@n5, Radius=@n6, Hardness=@n7), B=@n9)
  )
  FunctionOutput(None=@n10) {outputName=DissolveMask}
  FunctionOutput(None=@n9) {outputName=DissolveMaskEdgeSmoke}
  n11 = Constant() {r=1}
  n12 = Constant() {r=0.398}
  n13 = Constant() {r=1.5935}
  n14 = Subtract(A=@n12, B=Constant() {r=0.132})
  FunctionOutput(None=Subtract(A=SphereMask(A=@n10, B=@n11, Radius=@n12, Hardness=@n13), B=SphereMask(A=@n10, B=@n11, Radius=@n14, Hardness=@n13))) {outputName=DissolveMaskEdge2}
  n15 = ScalarParameter() {parameterName=DissolveMask_Radius}
  NamedRerouteDeclaration(None=FunctionInput(Preview=@n15) {bUsePreviewValueAsDefault=true, inputName=DissolveMask_Radius_1, inputType=FunctionInput_Scalar}) {name=DissolveMask_Radius, nodeColor=<rgba 0 0.7187 1 1>}
  NamedRerouteDeclaration(None=FunctionInput(Preview=TextureCoordinate()) {bUsePreviewValueAsDefault=true, inputType=FunctionInput_Vector2}) {name=IN_UV, nodeColor=<rgba 1 0.9609 0 1>}
  Subtract(A=@n15) {constB=0.3}
 ]}
