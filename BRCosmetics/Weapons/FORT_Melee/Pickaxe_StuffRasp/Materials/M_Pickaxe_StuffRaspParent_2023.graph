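# Parent material for the StuffRasp pickaxe that layers an animated smoky, glowing look between the shared 2023 character PreFX and PostFX functions.
# A dissolve mask, driven by swing-FX and black-style parameters, blends material attributes and adds tinted edge glow.
# A dark, rough "Xray" layer and emissive chain use Fresnels, panning world-aligned noise on pre-skinned positions, and distorted mask detail.
# A quality switch limits the flowing-smoke function to High and Epic.
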Material /BRCosmetics/Weapons/FORT_Melee/Pickaxe_StuffRasp/Materials/M_Pickaxe_StuffRaspParent_2023
graph {
 properties {
  bAutomaticallySetUsageInEditor=false
  bUseMaterialAttributes=true
  bUsedWithMeshDeformer=true
  bUsedWithMorphTargets=true
  bUsedWithSkeletalMesh=true
  pixelDepthOffsetMode=PDOM_Legacy
  shadingModel=MSM_FromMaterialExpression
  usedShadingModels=Default Lit | Subsurface | Clear Coat | Subsurface Profile | Cloth
 }
 nodes [
  Constant()
  LinearInterpolate(B=Subtract(A=ObjectPositionWS(), B=WorldPosition().XYZ), Alpha=ScalarParameter() {parameterName=HideMesh})
  n1 = NamedRerouteUsage()
  n2 = GetMaterialAttributes(None=NamedRerouteUsage()) {attributeGetTypes [#1]}
  n3 = NamedRerouteUsage()
  n4 = OneMinus(None=NamedRerouteUsage())
  n5 = Multiply(A=@n3, B=VectorParameter() {parameterName=SmokeGlowColor, defaultValue=<rgba 0.2342 0.3385 0.7305 1>}.RGB)
  n6 = NamedRerouteUsage()
  n7 = Multiply(A=VectorParameter() {parameterName=SmokeGlowColor2, defaultValue=<rgba 1.3913 2.4725 20 1>}.RGB)
  n8 = SetMaterialAttributes!(
   MaterialAttributes=@n2.MaterialAttributes
   Base Color=Add(A=@n2.Base Color, B=Multiply(A=@n3, B=@n4))
   Emissive Color=Add(A=Multiply(A=@n5, B=@n4), B=Multiply(A=@n6, B=@n7))
  ) {attributeSetTypes [#1, #2]}
  n9 = ScalarParameter() {defaultValue=1, group=DissolveMask, parameterName=DissolveMask_Radius}
  n10 = ScalarParameter() {group=StuffRasp_axe, parameterName=EnableSwingFX}
  n11 = Multiply(
   A=LinearInterpolate(B=Multiply(A=@n9, B=@n10), Alpha=ScalarParameter() {parameterName=BlackStyle}) {constA=1}
   B=ScalarParameter() {defaultValue=1, group=StuffRasp_axe, parameterName=BlackAuto}
  )
  n12 = MaterialFunctionCall!(
   DissolveMask_Radius_1=NamedRerouteDeclaration(None=Saturate(None=@n11)) {name=DissolveAnim, nodeColor=<rgba 1 0.6094 0 1>}
   In=NamedRerouteUsage()
  ) {materialFunction=&/MF_DissolveMask}
  n13 = NamedRerouteUsage()
  n14 = Multiply(A=Saturate(None=@n12.DissolveMask), B=Saturate(None=Multiply(A=@n13) {constB=10}))
  n15 = MaterialFunctionCall() {materialFunction=&MF_CharacterParent_2023_PreFX}
  MaterialFunctionCall(
   Material=SetMaterialAttributes!(MaterialAttributes=BlendMaterialAttributes(A=@n1, B=@n8, Alpha=@n14), World Position Offset) {attributeSetTypes [#3]}
   M=@n15.M
   Pre Skinned Local Position=@n15.M
   Distance Blend=@n15.M
   Fresnel=@n15.M
  ) {materialFunction=&MF_CharacterParent_2023_PostFX}
  n16 = Multiply(A=NamedRerouteUsage()) {constB=2}
  NamedRerouteDeclaration(None=TextureSample(UVs=Panner(Coordinate=@n16) {speedY=0.02}) {samplerType=SAMPLERTYPE_Grayscale, texture=/Game/Global/Textures/Default/Noise/LowResBlurredNoiseNormalized}.R) {name=GlobalNoise, nodeColor=<rgba 0.4531 0 1 1>}
  NamedRerouteDeclaration(None=Constant()) {name=Dot_Mask, nodeColor=<rgba 0 0.4609 1 1>}
  n17 = ScalarParameter() {group=TriplanarUV, parameterName=TriplanarUV_3D_UseClamp}
  n18 = ScalarParameter() {group=TriplanarUV, parameterName=TriplanarUV_3D_FlipUV}
  n19 = VectorParameter() {group=TriplanarUV, parameterName=TriplanarUV_3D_CenterPosition, defaultValue=<rgba 0 0 0 1>}
  n20 = ScalarParameter() {defaultValue=82, group=TriplanarUV, parameterName=TriplanarUV_3D_UVSize}
  NamedRerouteDeclaration(None=OneMinus(None=MaterialFunctionCall!(TriplanarUV_UseClamp?=@n17, TriplanarUV_FlipUV=@n18, TriplanarUV_CenterPosition=@n19.RGB, TriplanarUV_UVScale=@n20) {materialFunction=/BRCosmetics/Characters/Player/Male/Medium/Bodies/M_MED_BirchPalace/Skins/Gas/Materials/MF_TriplanarUV}.UV_FrontProjection)) {name=TriplannarUV_front, nodeColor=<rgba 1 0.3516 0 1>}
  NamedRerouteDeclaration(None=@n12.DissolveMaskEdgeSmoke) {name=DissolveMaskEdgeSmoke, nodeColor=<rgba 1 0.6563 0 1>}
  NamedRerouteDeclaration(None=@n12.DissolveMaskEdge2) {name=DissolveMaskEdge, nodeColor=<rgba 0.5703 1 0 1>}
  n21 = TextureSample() {
   samplerSource=SSM_Wrap_WorldGroupSettings
   samplerType=SAMPLERTYPE_Masks
   texture=/BRCosmetics/Weapons/FORT_Melee/Pickaxe_StuffRasp/Textures/T_Pickaxe_StuffRasp_TAMask
  }
  NamedRerouteDeclaration(None=@n21.R) {name=AO, nodeColor=<rgba 0 1 0.5781 1>}
  NamedRerouteDeclaration(None=@n21.G) {name=CRV, nodeColor=<rgba 1 0.2578 0 1>}
  n22 = NamedRerouteUsage()
  n23 = Fresnel(ExponentIn=ScalarParameter() {defaultValue=0.8156, group=StuffRasp_axe, parameterName=VertexNormal_FresnelPower}, Normal=VertexNormalWS())
  n24 = ScalarParameter() {defaultValue=-2, group=StuffRasp_axe, parameterName=PixNormal_FresnelPowerMin}
  n25 = ScalarParameter() {defaultValue=6.5, group=StuffRasp_axe, parameterName=PixNormal_FresnelPowerMax}
  n26 = TextureSampleParameter2D() {
   group=00_Texture Inputs
   parameterName=Mask
   samplerSource=SSM_Wrap_WorldGroupSettings
   samplerType=SAMPLERTYPE_Masks
   sortPriority=3
   texture=&T_BlackMask
  }
  n27 = Saturate(None=Fresnel(ExponentIn=LinearInterpolate(A=@n24, B=@n25, Alpha=@n26.G), Normal=PixelNormalWS()))
  NamedRerouteDeclaration(None=Add(A=Multiply(A=@n22, B=@n23), B=Power(Base=@n27) {constExponent=1.5})) {name=Main_Fresnnel, nodeColor=<rgba 0 0.0859 1 1>}
  n28 = TextureObjectParameter() {
   group=StuffRasp_axe
   parameterName=NoiseTexture
   samplerType=SAMPLERTYPE_LinearColor
   texture=/Game/Weapons/WeaponSkins/Textures/T_Swirly_Noise_highContrast1
  }
  n29 = VectorParameter() {group=StuffRasp_axe, parameterName=NoiseWorldTileScale, defaultValue=<rgba 111 111 50 0>}
  n30 = Subtract(
   A=VertexInterpolator(VS=LocalPosition() {localOrigin=InstancePreSkinning}.XYZ).PS
   B=Multiply(A=Time(), B=VectorParameter() {group=StuffRasp_axe, parameterName=NoisePanSpeed, defaultValue=<rgba -0.66 -1 -2 0>}.RGB)
  )
  n31 = VertexInterpolator(VS=PreSkinnedNormal())
  NamedRerouteDeclaration(None=ComponentMask(None=MaterialFunctionCall(TextureObject=@n28, TextureSize=@n29.RGB, WorldPosition=@n30, World Space Normal=@n31.PS) {materialFunction=/Game/Weapons/WeaponSkins/Wraps/MaterialFunctions/WorldAlignedTexture}.XYZ Texture) {r=true}) {name=SmokeNoise, nodeColor=<rgba 1 0 0.3281 1>}
  n32 = Constant() {r=1}
  n33 = Multiply(
   A=MaterialFunctionCall!(In_FireSmoke_TAMask, In_FireSmoke_UV=NamedRerouteUsage()) {materialFunction=/BRCosmetics/Gadgets/Assets/VinderTech_GliderChute/Glider_RampCanal/Materials/MFW_FlowingSmoke}.Result
   B=ScalarParameter() {defaultValue=0.5, group=StuffRasp_axe, parameterName=DetailSmoke_Intensity}
  ) {constB=0.5}
  n34 = MaterialFunctionCall!(
   Low _ PC Low and Med=@n32
   "High _ XB1, PS4"=@n33
   "Epic _ PC High, XSX, PS5"=@n33
   --------
   "Mobile Low _ PC Perf, Android"=@n32
   Mobile Med _ Switch=@n32
  ) {materialFunction=/Game/Global/MaterialFunctions/TechArt/MF_Fortnite_QualitySwitch_Scalar}
  n35 = Multiply(A=NamedRerouteUsage(), B=Constant() {r=20}) {constB=2}
  n36 = Multiply(
   A=TextureSample(UVs=Panner(Coordinate=@n35) {speedY=0.4}) {samplerType=SAMPLERTYPE_Grayscale, texture=&LowResBlurredNoiseNormalized}.R
   B=NamedRerouteUsage()
  )
  n37 = Constant() {r=0.02}
  n38 = TextureSample(UVs=Add(A=TextureCoordinate(), B=Multiply(A=@n36, B=@n37))) {
   samplerSource=SSM_Wrap_WorldGroupSettings
   samplerType=SAMPLERTYPE_Masks
   texture=&T_Pickaxe_StuffRasp_TAMask
  }
  n39 = ScalarParameter() {defaultValue=3.5, group=StuffRasp_axe, parameterName=DetailSmokeBrightness}
  n40 = ComponentMask(None=Power(Base=Multiply(A=@n38.R, B=@n39), Exp=ScalarParameter() {defaultValue=0.85, group=StuffRasp_axe, parameterName=DetailSmokeBrightnessPower})) {r=true}
  n41 = NamedRerouteUsage()
  n42 = LinearInterpolate(A=Constant() {r=1}, B=Constant() {r=-1}, Alpha=Power(Base=ComponentMask(None=@n41) {g=true}) {constExponent=3})
  NamedRerouteDeclaration(None=Multiply(A=Multiply(A=@n34.Result, B=@n40), B=Saturate(None=@n42)) {constB=0.5}) {name=DetailSmoke, nodeColor=<rgba 1 0 0.8906 1>}
  n43 = Multiply(
   A=Fresnel(ExponentIn=ScalarParameter() {defaultValue=2.1, group=StuffRasp_axe, parameterName=FresnelGlobal}, Normal=VertexNormalWS())
   B=Constant() {r=0.8}
  )
  n44 = Constant() {r=0.1}
  n45 = NamedRerouteUsage()
  n46 = Multiply(A=NamedRerouteUsage(), B=LinearInterpolate(A=Constant() {r=1.3}, B=Constant() {r=2.2}, Alpha=NamedRerouteUsage()))
  n47 = Constant() {r=10}
  n48 = Fresnel(ExponentIn=Constant() {r=-0.1422}, Normal=PixelNormalWS())
  n49 = Multiply(
   A=NamedRerouteUsage()
   B=LinearInterpolate(
    A=ScalarParameter() {defaultValue=0.3, group=StuffRasp_axe, parameterName=NoiseScaleOuter}
    B=ScalarParameter() {defaultValue=0.75, group=StuffRasp_axe, parameterName=NoiseScaleInner}
    Alpha=Saturate(None=@n48)
   )
  )
  n50 = NamedRerouteUsage()
  n51 = Constant() {r=2}
  n52 = Max(
   A=Multiply(A=LinearInterpolate(A=@n44, Alpha=@n45) {constB=0.8}, B=Power(Base=@n46, Exp=@n47))
   B=Multiply(A=Saturate(None=@n49), B=Multiply(A=@n50, B=@n51))
  )
  n53 = NamedRerouteUsage()
  n54 = Multiply(A=Multiply(A=NamedRerouteUsage(), B=NamedRerouteUsage()), B=Saturate(None=Multiply(A=@n53) {constB=2}))
  n55 = LinearInterpolate(B=Constant(), Alpha=@n54) {constA=1, constB=0}
  n56 = LinearInterpolate(B=Constant() {r=2}, Alpha=@n54) {constB=0}
  n57 = NamedRerouteUsage()
  n58 = Add(A=NamedRerouteUsage(), B=Multiply(A=Add(A=@n43, B=@n52), B=LinearInterpolate(A=@n55, B=@n56, Alpha=@n57)))
  n59 = ScalarParameter() {defaultValue=1.6, group=StuffRasp_axe, parameterName=GlobalContrast}
  n60 = Constant() {r=1}
  n61 = VectorParameter() {group=StuffRasp_axe, parameterName=Stuffrasppickaxe_EmissiveTint, defaultValue=<rgba 1.1 0.055 0.1652 1>}
  n62 = ScalarParameter() {group=StuffRasp_axe, parameterName=Veil}
  NamedRerouteDeclaration(None=Multiply(A=Power(Base=@n58, Exp=@n59), B=LinearInterpolate(A=@n60, B=@n61.RGB, Alpha=@n62))) {name=OutEmissvie, nodeColor=<rgba 1 0 0.9844 1>}
  NamedRerouteDeclaration(None=@n15.Result) {name=In_MA, nodeColor=<rgba 0.1953 1 0 1>}
  n63 = SetMaterialAttributes!(
   MaterialAttributes=GetMaterialAttributes(None=NamedRerouteUsage()) {attributeGetTypes [#2]}.MaterialAttributes
   Emissive Color=NamedRerouteUsage()
   Base Color=Constant()
   Metallic=Constant() {r=0.2}
   Specular=Constant() {r=0.02}
   Roughness=Constant() {r=1}
   Anisotropy
  ) {attributeSetTypes [
    #2
    #1
    #4
    #5
    #6
    #7
   ]}
  NamedRerouteDeclaration(None=@n63) {name=Xray_Mat, nodeColor=<rgba 1 0 0.5156 1>}
 ]
}
