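# Animation Blueprint for the StuffRasp pickaxe mesh, a cosmetic melee weapon.
# The AnimGraph blends by an unwired bool between the incoming input pose and a single-state machine that plays the StuffRasp idle sequence.
# The update event caches the part mesh's attach parent as Body_Mesh and stores world rotations of its two right lower-arm twist sockets.
# Those stored twist rotations are not read by the AnimGraph.
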AnimBlueprint /BRCosmetics/Weapons/FORT_Melee/Pickaxe_StuffRasp/Meshes/StuffRasp_Axe_AnimBP
target_skeleton=&/StuffRasp_Axe_Skeleton
graphs [
 AnimGraph {nodes [
   AnimGraphNode_StateMachine_1 = Idle Pose() {is=title AnimGraphNode_StateMachine}
   AnimGraphNode_SubInput_0 = Input Pose()
   Output Pose(Result=Blend Poses by bool(BlendPose_0=@AnimGraphNode_SubInput_0, BlendPose_1=@AnimGraphNode_StateMachine_1) {blendType=Linear}) {layerGroup=None}
  ]}
 Idle Pose {entry=Idle, path=AnimGraph.AnimGraphNode_StateMachine_1.Idle Pose, ids=id, nodes [AnimStateNode_0 = Idle()]}
 Idle {
  path=AnimGraph.AnimGraphNode_StateMachine_1.Idle Pose.AnimStateNode_0.Idle
  nodes [Output Animation Pose(Result=Sequence Player 'StruffRasp_Idle'() {sequence=/BRCosmetics/Animation/Game/MainPlayer/Combat/Melee/DualWield/StuffRasp/StruffRasp_Idle})]
 }
 EventGraph {nodes [
   Event BlueprintUpdateAnimation()
   n1 = GetPartSkeletalMeshComponent()
   > Is Valid(InputObject=@n1)
   n2 = GetAttachParent!(self:Scene Component Object Reference=@n1) -> (ReturnValue:Scene Component Object Reference)
   > Is Valid(InputObject=@n2)
   > Set Body_Mesh(Body_Mesh:Scene Component Object Reference=@n2) -> (Output_Get:Scene Component Object Reference)
   n3 = Get Body_Mesh() -> (Body_Mesh:Scene Component Object Reference)
   > Set Twist_01(Twist_01:Rotator=BreakTransform(InTransform=GetSocketTransform(self=@n3, InSocketName=lowerarm_twist_01_r, TransformSpace=RTS_World)).Rotation) -> (Output_Get:Rotator)
   > Set Twist_02(Twist_02:Rotator=BreakTransform(InTransform=GetSocketTransform(self=@n3, InSocketName=lowerarm_twist_02_r, TransformSpace=RTS_World)).Rotation) -> (Output_Get:Rotator)
   TryGetPawnOwner()
  ]}
]
variables [
 Body_Mesh {type=SceneComponent, default=null}
 Twist_01 {type=Rotator}
 Twist_02 {type=Rotator}
]
